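# Unlit translucent material for Niagara mesh particles in the SparkleChop skydiving contrail effect.
# Particle color passes through a shared emissive adjustment function.
# A camera-facing falloff from the normal-view dot product, toggled by the UseGlow scalar against solid, multiplies particle alpha for soft glow edges.
# Vertex color red and green mask two Niagara dynamic parameters into a summed Z-axis vector, likely a per-vertex offset.
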Material /BRCosmetics/Athena/Effects/Particles/Skydiving/NS_Contrail_SparkleChop/M_Contrail_SparkleChop
graph {
 properties {
  bUsedWithNiagaraMeshParticles=true
  blendMode=BLEND_Translucent
  pixelDepthOffsetMode=PDOM_Legacy
  shadingModel=MSM_Unlit
  usedShadingModels=Unlit
 }
 nodes [
  n1 = Constant2Vector()
  n2 = DynamicParameter()
  n3 = VertexColor()
  n4 = Constant2Vector()
  Add(A=Multiply(A=AppendVector(A=@n1, B=@n2.Param1), B=@n3.R), B=Multiply(A=@n3.G, B=AppendVector(A=@n4, B=@n2.Param2)))
  Constant3Vector() {constant=<rgba 1 1 1 0>}
  MaterialFunctionCall(Emissive=ParticleColor().RGB) {materialFunction=&MF_FX_EmissiveAdjustments}
  n5 = VertexNormalWS()
  n6 = CameraVectorWS()
  n7 = Divide(A=Power(Base=DotProduct(A=@n5, B=@n6)) {constExponent=6}) {constB=3}
  Multiply(A=LinearInterpolate(B=Saturate(None=@n7), Alpha=ScalarParameter() {parameterName=UseGlow}) {constA=1}, B=ParticleColor().A)
 ]
}
