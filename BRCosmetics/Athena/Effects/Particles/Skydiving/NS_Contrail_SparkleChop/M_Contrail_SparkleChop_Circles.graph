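# Translucent unlit Niagara sprite material that draws procedural circles, named for the SparkleChop skydiving contrail effect.
# Distance from the UV center forms a hard-edged disc, with a dynamic parameter widening an inner cutout to hollow it into a thinning ring.
# The IsGlow scalar blends that shape toward a soft, power-sharpened radial glow.
# Particle color feeds the shared MF_FX_EmissiveAdjustments function.
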Material /BRCosmetics/Athena/Effects/Particles/Skydiving/NS_Contrail_SparkleChop/M_Contrail_SparkleChop_Circles
graph {
 properties {
  bAutomaticallySetUsageInEditor=false
  bUsedWithNiagaraSprites=true
  blendMode=BLEND_Translucent
  pixelDepthOffsetMode=PDOM_Legacy
  shadingModel=MSM_Unlit
  usedShadingModels=Unlit
 }
 nodes [
  n1 = Length(None=Subtract(A=TextureCoordinate()) {constB=0.5}) {is=class MaterialExpressionLength}
  n2 = LinearInterpolate(Alpha=DynamicParameter() {defaultValue=<rgba 0 0 1 1>}.Param1) {constB=0.35}
  LinearInterpolate(
   A=Subtract(A=Step(Y=@n1) {constX=0.35}, B=Step(Y=@n1, X=@n2) {constX=0.1})
   B=Power(Base=OneMinus(None=@n1)) {constExponent=9}
   Alpha=ScalarParameter() {parameterName=IsGlow}
  )
  MaterialFunctionCall(Emissive=ParticleColor().RGB) {materialFunction=&MF_FX_EmissiveAdjustments}
 ]
}
