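# Lit material for square sparkle particles in the SparkleChop skydiving contrail, usable on Niagara sprites and mesh particles.
# UVs snap to a 12-cell grid and pan, giving blocky, shifting color from a color texture multiplied by chromatic noise.
# A tiled quad texture, thresholded by a step, masks the color texture's alpha, alongside a matching normal map.
# Dynamic parameters weighted by vertex color red and green build an offset blended by the UseWPO parameter.
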Material /BRCosmetics/Athena/Effects/Particles/Skydiving/NS_Contrail_SparkleChop/M_Contrail_SparkleChop_Square
graph {
 properties {bUsedWithNiagaraMeshParticles=true, bUsedWithNiagaraSprites=true, pixelDepthOffsetMode=PDOM_Legacy, usedShadingModels=Default Lit}
 nodes [
  n1 = DynamicParameter()
  n2 = AppendVector(A=Constant2Vector(), B=@n1.R)
  n3 = VertexColor()
  n4 = AppendVector(A=Constant2Vector(), B=@n1.G)
  LinearInterpolate(B=Add(A=Multiply(A=@n2, B=@n3.R), B=Multiply(A=@n3.G, B=@n4)), Alpha=ScalarParameter() {parameterName=UseWPO})
  n5 = Multiply(A=TextureCoordinate()) {constB=12}
  n6 = Panner(Coordinate=Divide(A=Floor(None=@n5)) {constB=12}) {speedX=0.1, speedY=0.2}
  n7 = TextureSample(UVs=@n6) {texture=&/T_Contrail_SparkleChop_Color}
  n8 = TextureSample(UVs=Multiply(A=TextureCoordinate()) {constB=4}) {
   samplerType=SAMPLERTYPE_LinearGrayscale
   texture=&/T_Contrail_SparkleChop_Quads
  }
  LinearInterpolate(A=Constant(), B=@n7.A, Alpha=Step(X=@n8.R) {constY=0.2})
  Multiply(A=Multiply(
    A=TextureSample(UVs=@n6) {texture=&/T_DirectionalNoise_Chromatic}.RGB
    B=@n7.RGB
   )) {constB=3}
  Time()
  n9 = TextureSample(UVs=Multiply(A=TextureCoordinate()) {constB=4}) {
   samplerType=SAMPLERTYPE_Normal
   texture=&/T_Contrail_SparkleChop_Quads_Normal
  }
  Transform(None=@n9.RGB)
 ]
}
