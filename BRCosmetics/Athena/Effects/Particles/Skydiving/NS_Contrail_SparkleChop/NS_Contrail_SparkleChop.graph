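# Niagara skydiving contrail cosmetic "SparkleChop", layering colored sparkle sprites, disco streaks, speed lines, light and lightning meshes, flares with lights, and spinning stars.
# Nine emitters sample the skeletal mesh, most using partition or spawn-group indices to pick bones or sockets.
# Glide lean alpha drives spawn rates, streak length, spin and playback, while dive state feeds ray materials and alpha.
# Front-end and time-of-day modifiers adjust scale, orientation and brightness, and distance fraction gates ray growth.
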NiagaraSystem /BRCosmetics/Athena/Effects/Particles/Skydiving/NS_Contrail_SparkleChop/NS_Contrail_SparkleChop
dependencies {
 usedDataInterfaces [
  &NiagaraDataInterfaceArrayColor
  &NiagaraDataInterfaceCurve
  &NiagaraDataInterfaceArrayFloat3
  &NiagaraDataInterfaceArrayFloat
  &NiagaraDataInterfaceArrayInt32
 ]
 usedDynamicInputs []
 usedModules [
  &self:SystemSpawnScript.SystemScriptSource.SystemScriptGraph.NiagaraNodeAssignment_0.SetVariables_#1
  &SystemState
 ]
}
emitters {
 Decal {
  emitterUpdateScript [
   &EmitterState()
   &SpawnRate(SpawnRate=&Lerp_Float(A=10, B=20, Alpha=$User.ParaGlideLeanAlpha), Use Spawn Probability=false)
  ]
  particleSpawnScript [
   /Niagara/Modules/Spawn/Initialization/InitializeParticle(
    Color=<rgba 0.9062 0.4123 0.1067 0.6>
    Lifetime=0.6
    Module.Write Color=true
    Module.Write Lifetime=true
    Module.Write Mass=true
    Module.Write Material Random=true
    Module.Write Position=false
    Module.Write Scale=false
    Module.Write SpriteRotation=false
    Module.Write SpriteSize=true
    Position=&SimulationPosition()
    Sprite Size=<40 40>
   ) {enabled=false}
   /Niagara/Modules/Spawn/Initialization/V2/InitializeParticle(
    Color=&SelectLinearColorFromArray(
     Array Sampling Mode=::NewEnumerator0
     Direct Array Index=?
     Color Selection Array=DataInterface() {propertyValues {ColorData [<rgba 0.0459 0.0386 0.1667 0.8>, <rgba 0.3489 0.2624 0.8177 0.8>, <rgba 0.7031 0.0174 0.2835 0.8>, <rgba 0.0116 0.2817 0.6979 0.8>]}}
     Random Seed=0
    )
    Lifetime Max=0.4
    Lifetime Min=0.3
    Lifetime Mode=::NewEnumerator1
    Mass Mode=::NewEnumerator1
    Sprite Size=<80 80>
    Sprite Size Mode=::NewEnumerator4
    UsePositionOffset=false
    Write Material Random=true
   ) {module=InitializeParticle001}
   &SkeletalMeshLocation(
    Inherit Velocity=false
    Interpolate Mesh Sampling=false
    MaskSpawnGroup=false
    Module.Override Randomness=false
    Skeletal Mesh=$System.Skeletal Mesh
   )
   &CameraOffset(Camera Offset Amount=9)
  ]
  particleUpdateScript [
   &ParticleState()
   &ScaleColor(
    Scale Alpha=&FloatFromCurve(
     FloatCurve=DataInterface() {propertyValues {Curve {keys [
         {interpMode=RCIM_Cubic, tangentMode=RCTM_User, time=0, value=0}
         {interpMode=RCIM_Cubic, tangentMode=RCTM_User, time=0.5, value=1}
         {tangentMode=RCTM_User, time=1, value=0}
        ]}}}
     Scale Curve=&FN_FrontEndFloatModifier(Gameplay=1, Front End=0.4)
    )
    ScaleA=true
    ScaleRGB=true
   )
  ]
  rendererClasses [&NiagaraSpriteRendererProperties]
  renderers [{rendererClass=&NiagaraSpriteRendererProperties}]
 }
 DiscoStreaks {
  emitterUpdateScript [
   &EmitterState()
   /Game/Effects/Niagara/Modules/Spawn/SpawnRateFromMultipleSpawnGroups_2(Global Spawnrate=4)
  ]
  particleSpawnScript [
   &PartitionParticles(Number Of Partitions=4)
   &FN_SetSpawnGroup(Number of Spawn Groups=2)
   &InitializeParticle(
    Color=&SelectLinearColorFromArray(
     Array Sampling Mode=::NewEnumerator0
     Direct Array Index=?
     Color Selection Array=DataInterface() {propertyValues {ColorData [<rgba 0.0459 0.0386 0.1667 1>, <rgba 0.3489 0.2624 0.8177 1>, <rgba 0.7031 0.0174 0.2835 0.979>, <rgba 0.0116 0.2817 0.6979 0.974>]}}
     Random Seed=0
    )
    Lifetime Max=1
    Lifetime Min=0.3
    Lifetime Mode=::NewEnumerator1
    Mesh Scale Mode=::NewEnumerator1
    Mesh Uniform Scale Max:NiagaraFloat=&RandomRangeFloat(Minimum=0.035, Maximum=0.065)
    Mesh Uniform Scale Min=0.02
    UsePositionOffset=false
    Write Material Random=true
   ) {module=InitializeParticle001}
   &SkeletalMeshLocation(
    Bone / Socket Index=$Particles.PartitionParticles.Partition
    Bone Sampling Mode=::NewEnumerator4
    Inherit Velocity=false
    MaskSpawnGroup=false
    Skeletal Mesh=$System.Skeletal Mesh
   )
   &ShapeLocation(MaskSpawnGroup=false, Override Seed=false, Ring Radius=15, Shape Primitive=::NewEnumerator4, UseOwnerScale=true)
  ]
  particleUpdateScript [
   &ParticleState()
   &DynamicMaterialParameters(
    Index 0 Param 1=&FloatFromCurve(
     FloatCurve=DataInterface() {propertyValues {Curve {keys [
         {interpMode=RCIM_Cubic, leaveTangent=-23.0281, tangentMode=RCTM_Break, time=0, value=0}
         {arriveTangent=31.7752, interpMode=RCIM_Cubic, leaveTangent=153.4346, tangentMode=RCTM_Break, time=0.4, value=15}
         <at 1 55>
        ]}}}
     Scale Curve=1.5
    )
    Index 0 Param 2=&FloatFromCurve(FloatCurve=DataInterface() {propertyValues {Curve {keys [<at 0.4 0>,<at 0.9 55>]}}}, Scale Curve=1.5)
    Module.Param0WriteEnabled=true
    Module.Param1WriteEnabled=true
    Module.Param2WriteEnabled=true
    Module.Param3WriteEnabled=true
   )
   /SetVariables(
    Particles.GlowColor=&MultiplyLinearColor(Linear Color=$Particles.Initial.Color, Scale Factor=<3 3 3 0.5>)
    Particles.GlowSize:Vector3f=&Multiply_VectorByFloat(Vector=$Particles.Initial.Scale, Float=3)
   ) {isSetParameters=true}
  ]
  rendererClasses [&NiagaraMeshRendererProperties]
  renderers [
   {rendererClass=&NiagaraMeshRendererProperties, rendererIndex=0}
   {rendererClass=&NiagaraMeshRendererProperties, rendererIndex=1}
  ]
 }
 Flare {
  emitterUpdateScript [
   &EmitterState(Inactive Response=::NewEnumerator1, Life Cycle Mode=::NewEnumerator1, Loop Duration=0.1, UseLoopDelay=false)
   &SpawnBurst_Instantaneous(Spawn Count=2, Use Spawn Probability=false, UseLoopCountLimit=false)
  ]
  particleSpawnScript [
   &PartitionParticles()
   &InitializeParticle(
    Color=<rgba 0.4583 0.0219 0.2416 1>
    Lifetime=0.4
    Sprite Rotation Mode=::NewEnumerator2
    Sprite Size=&Multiply_Vector2D_ByFloat(
     Float=&Multiply_Float(
      A=&RandomRangeFloat(Minimum=20, Maximum=25)
      B=&MakeCustomFloatFromBool(
       Bool=&SetBoolByIntComparison(A=&SpawnGroup(), B=2, Comparison Type=::NewEnumerator2)
       True Float=0.75
       False Float=1
      )
     )
     Vector2D=<1 1.5>
    )
    Sprite Size Mode=::NewEnumerator4
    UsePositionOffset=false
    Write Material Random=true
   )
   &CameraOffset(Camera Offset Amount=15)
  ]
  particleUpdateScript [
   &ParticleState()
   &SkeletalMeshLocation(
    Bone / Socket Index=$Particles.PartitionParticles.Partition
    Bone Sampling Mode=::NewEnumerator4
    MaskSpawnGroup=false
    Skeletal Mesh=$System.Skeletal Mesh
   )
   &ScaleSpriteSize(Uniform Curve Sprite Scale=DataInterface() {propertyValues {Curve {keys [
        {interpMode=RCIM_Cubic, tangentMode=RCTM_User, time=0, value=0}
        {interpMode=RCIM_Cubic, tangentMode=RCTM_User, time=0.5, value=1}
        {tangentMode=RCTM_User, time=1, value=0}
       ]}}})
   &FN_Light_Attributes(
    Color=<rgba 0.4583 0 0.0764 1>
    Create Color Attribute=true
    Create Exponent Attribute=true
    Create Radius Attribute=true
    Exponent=4
    Radius=50
    Scale Color=50
   )
   &ScaleColor(
    Scale RGB=&VectorFromFloat(Value=&FNTimeOfDayScalar(Night Scale=0.5, Morning Scale=0.6, Evening Scale=0.6))
    ScaleA=true
    ScaleRGB=true
   )
  ]
  rendererClasses [&NiagaraSpriteRendererProperties, &NiagaraLightRendererProperties]
  renderers [
   {rendererClass=&NiagaraSpriteRendererProperties}
   {rendererClass=&NiagaraLightRendererProperties, rendererIndex=1}
  ]
 }
 LightRay {
  emitterUpdateScript [
   &EmitterState(
    Inactive Response=::NewEnumerator1
    Life Cycle Mode=::NewEnumerator1
    Loop Behavior=::NewEnumerator1
    Loop Duration Mode=::NewEnumerator1
    UseLoopDelay=false
   )
   &SpawnRate(SpawnRate=4, Use Spawn Probability=false) {enabled=false}
   /SetVariables(Emitter.DistanceFractionThreshold:NiagaraBool=&SetBoolByFloatComparison(A=$Emitter.Scalability.DistanceFraction, B=0.9, Comparison Type=::NewEnumerator6)) {isSetParameters=true, module=SetVariables_#2}
   &SpawnBurst_Instantaneous(Spawn Count=2, Use Spawn Probability=false, UseLoopCountLimit=false)
   &Timeline(Play=$Emitter.DistanceFractionThreshold, Play Rate=4, Rewind Rate=4) {module=Timeline001}
  ]
  particleSpawnScript [
   &PartitionParticles()
   &InitializeParticle(
    Color Mode=::NewEnumerator0
    Lifetime=&RandomRangeFloat(Minimum=1.7, Maximum=2)
    Mesh Scale Mode=::NewEnumerator1
    Mesh Uniform Scale Max=1
    Mesh Uniform Scale Min=0.8
    UsePositionOffset=false
    Write Material Random=true
   )
   &InitialMeshOrientation(
    Mesh Axis To Orient=&FN_FrontEndVector3Modifier(Gameplay=<1 0 0>, Front End=<1 0 0>)
    Mesh Orientation Mode=::NewEnumerator2
    Orientation Coordinate Space=::Local
    Rotation=&FN_FrontEndVector3Modifier(
     Gameplay=&SelectVectorFromArray(
      Direct Array Index=$Particles.PartitionParticles.Partition
      Vector Selection Array=DataInterface() {propertyValues {FloatData [<0.03 0.01 0.09>, <0.03 -0.01 -0.09>, <0 0 0>, <0 0 0>]}}
     )
     Front End=&SelectVectorFromArray(
      Direct Array Index=$Particles.PartitionParticles.Partition
      Vector Selection Array=DataInterface() {propertyValues {FloatData [<0.05 0.03 0.09>, <0.05 -0.03 -0.09>, <0 0 0>, <0 0 0>]}}
     )
    )
    Rotation Coordinate Space=::NewEnumerator2
    Use Rotation Vector=false
   ) {module=InitialMeshOrientation001}
  ]
  particleUpdateScript [
   &ParticleState(Kill Particles When Lifetime Has Elapsed=false, Let Infinitely Lived Particles Die When Emitter Deactivates=true)
   &UpdateMeshOrientation(
    Facing Direction=&Lerp_Vector(
     A=&SelectVectorFromArray(
      Direct Array Index=$Particles.PartitionParticles.Partition
      Vector Selection Array=DataInterface() {propertyValues {FloatData [<1 0 0>, <1 0 0>]}}
     )
     B=&SelectVectorFromArray(
      Direct Array Index=$Particles.PartitionParticles.Partition
      Vector Selection Array=DataInterface() {propertyValues {FloatData [<1 0 0>, <1 0 0>]}}
     )
     Alpha=$User.ParaGlideLeanAlpha
    )
    Facing Interpolation=false
    Orientation Method=::NewEnumerator1
    Override Quat=false
    TransformFacing=false
    TransformUp=false
    Up Direction:Vector3f=&FN_FrontEndVector3Modifier(Gameplay=<0 0 1>, Front End=&Lerp_Vector(A=<0 0 1>, B=<0 -0.5 1>, Alpha=$User.ParaGlideLeanAlpha))
   ) {module=UpdateMeshOrientation002}
   &Timeline(
    Play Rate=&Lerp_Float(A=1, B=2, Alpha=&ClampFloat(Float=$User.ParaGlideLeanAlpha))
    Timeline End=10000000
   )
   &ScaleMeshSize(Scale Factor=&MakeVector(
     X=1
     Y=1
     Z=&FloatFromCurve(
      FloatCurve=DataInterface() {propertyValues {Curve {keys [
          {arriveTangent=5.7371, interpMode=RCIM_Cubic, leaveTangent=5.7371, tangentMode=RCTM_User, time=0, value=0}
          {arriveTangent=-0.9106, interpMode=RCIM_Cubic, leaveTangent=-0.9106, tangentMode=RCTM_User, time=0.76, value=1.025}
          <at 1 1>
         ]}}}
      CurveIndex=$Emitter.Age
      Scale Curve=$Emitter.Timeline001.NormalizedTime
     )
    ))
   &DynamicMaterialParameters(
    Index 0 Param 2=$System.Dive
    Index 0 Param 3=&RandomRangeFloat(Minimum=0, Maximum=1)
    Index 0 Param 4=&RandomRangeFloat(Minimum=0, Maximum=1)
    Module.Index0Param0Write=true
    Module.Index0Param1Write=true
    Module.Index0Param2Write=true
    Module.Index0Param3Write=true
   ) {module=DynamicMaterialParameters001}
   &SkeletalMeshLocation(
    Bone / Socket Index=$Particles.PartitionParticles.Partition
    Bone Sampling Mode=::NewEnumerator4
    MaskSpawnGroup=false
    Orientation Sampling=::NewEnumerator1
    Sampled Position Offset=&FN_FrontEndVector3Modifier(
     Gameplay=&MakeCustomVectorFromBool(
      Bool=$User.bIsFrontEndPreview
      True Vector=&SelectVectorFromArray(
       Direct Array Index=$Particles.PartitionParticles.Partition
       Vector Selection Array=DataInterface() {propertyValues {FloatData [<0 -3 1>, <0 4 1>, <0 0 0>, <0 0 0>]}}
      )
      False Vector=&Lerp_Vector(
       A=&SelectVectorFromArray(
        Direct Array Index=$Particles.PartitionParticles.Partition
        Vector Selection Array=DataInterface() {propertyValues {FloatData [<8 -3 -1>, <-8 4 1>, <0 0 0>, <0 0 0>]}}
       )
       B=&SelectVectorFromArray(
        Direct Array Index=$Particles.PartitionParticles.Partition
        Vector Selection Array=DataInterface() {propertyValues {FloatData [<25 0 0>, <-25 0 0>, <0 0 0>, <0 0 0>]}}
       )
       Alpha=$User.ParaGlideLeanAlpha
      )
     )
     Front End=&Lerp_Vector(
      A=&SelectVectorFromArray(
       Direct Array Index=$Particles.PartitionParticles.Partition
       Vector Selection Array=DataInterface() {propertyValues {FloatData [<0 -3 -1>, <0 4 1>, <0 0 0>, <0 0 0>]}}
      )
      B=&SelectVectorFromArray(
       Direct Array Index=$Particles.PartitionParticles.Partition
       Vector Selection Array=DataInterface() {propertyValues {FloatData [<25 0 0>, <-25 0 0>, <0 0 0>, <0 0 0>]}}
      )
      Alpha=$User.ParaGlideLeanAlpha
     )
    )
    Skeletal Mesh=$System.Skeletal Mesh
   ) {enabled=false}
   &SkeletalMeshLocation(
    Bone / Socket Index=$Particles.PartitionParticles.Partition
    Bone Sampling Mode=::NewEnumerator4
    MaskSpawnGroup=false
    Orientation Sampling=::NewEnumerator1
    Sampled Position Offset=<0 0 5>
    Sampled Position Offset Coordinate Space:ENiagaraSkelMeshTransforms=::NewEnumerator0
    Skeletal Mesh=$System.Skeletal Mesh
   ) {module=SkeletalMeshLocation001}
   &ScaleMeshSize(Scale Factor=&FN_FrontEndVector3Modifier(Gameplay=<0.7 0.7 0.8>, Front End=<0.65 0.65 0.75>)) {module=ScaleMeshSize001}
   &ScaleMeshSize(Scale Factor=&SelectVectorFromArray(
     Direct Array Index=$Particles.PartitionParticles.Partition
     Vector Selection Array=DataInterface() {propertyValues {FloatData [<1 1 1>, <1 1 1>, <0.5 0.5 0.4>, <0.5 0.5 0.4>]}}
    )) {module=ScaleMeshSize002}
   &ScaleColor(
    Scale Alpha=&FloatFromCurve(
     FloatCurve=DataInterface() {propertyValues {Curve {keys [<at 0 1 RCIM_Cubic>,<at 0.2 1 RCIM_Cubic>,<at 1 1>]}}}
     Scale Curve=&Lerp_Float(A=0.1, B=0.2, Alpha=$System.Dive)
    )
    Scale RGB=&MakeVectorFromLinearColorRGB(Color=&FNTimeOfDayColor(
      Night Color=<rgba 0.6 0.6 0.6 1>
      Morning Color=<rgba 0.65 0.65 0.65 1>
      Day Color=<rgba 0.9 0.9 0.9 1>
      Evening Color=<rgba 0.65 0.65 0.65 1>
     ))
    ScaleA=true
    ScaleRGB=true
   )
   &ScaleColor(Scale Alpha=&FN_FrontEndFloatModifier(Gameplay=1, Front End=0.6), ScaleA=true, ScaleRGB=true) {module=ScaleColor001}
  ]
  rendererClasses [&NiagaraMeshRendererProperties]
  renderers [{rendererClass=&NiagaraMeshRendererProperties, rendererIndex=0}]
 }
 LightningRay {
  emitterUpdateScript [
   &EmitterState(
    Inactive Response=::NewEnumerator1
    Life Cycle Mode=::NewEnumerator1
    Loop Behavior=::NewEnumerator1
    Loop Duration Mode=::NewEnumerator1
    UseLoopDelay=false
   )
   &SpawnRate(SpawnRate=4, Use Spawn Probability=false) {enabled=false}
   /SetVariables(Emitter.DistanceFractionThreshold:NiagaraBool=&SetBoolByFloatComparison(A=$Emitter.Scalability.DistanceFraction, B=0.9, Comparison Type=::NewEnumerator6)) {isSetParameters=true, module=SetVariables_#2}
   &SpawnBurst_Instantaneous(Spawn Count=2, Use Spawn Probability=false, UseLoopCountLimit=false)
   &Timeline(Play=$Emitter.DistanceFractionThreshold, Play Rate=4, Rewind Rate=4) {module=Timeline001}
  ]
  particleSpawnScript [
   &PartitionParticles()
   &InitializeParticle(
    Color Mode=::NewEnumerator0
    Lifetime=&RandomRangeFloat(Minimum=1.7, Maximum=2)
    Mesh Scale Mode=::NewEnumerator1
    Mesh Uniform Scale Max=1
    Mesh Uniform Scale Min=0.8
    UsePositionOffset=false
    Write Material Random=true
   )
   &InitialMeshOrientation(
    Mesh Axis To Orient=&FN_FrontEndVector3Modifier(Gameplay=<1 0 0>, Front End=<1 0 0>)
    Mesh Orientation Mode=::NewEnumerator2
    Orientation Coordinate Space=::Local
    Rotation=&FN_FrontEndVector3Modifier(
     Gameplay=&SelectVectorFromArray(
      Direct Array Index=$Particles.PartitionParticles.Partition
      Vector Selection Array=DataInterface() {propertyValues {FloatData [<0.03 0.01 0.09>, <0.03 -0.01 -0.09>, <0 0 0>, <0 0 0>]}}
     )
     Front End=&SelectVectorFromArray(
      Direct Array Index=$Particles.PartitionParticles.Partition
      Vector Selection Array=DataInterface() {propertyValues {FloatData [<0.05 0.03 0.09>, <0.05 -0.03 -0.09>, <0 0 0>, <0 0 0>]}}
     )
    )
    Rotation Coordinate Space=::NewEnumerator2
    Use Rotation Vector=true
   ) {module=InitialMeshOrientation001}
  ]
  particleUpdateScript [
   &ParticleState(Kill Particles When Lifetime Has Elapsed=false, Let Infinitely Lived Particles Die When Emitter Deactivates=true)
   &UpdateMeshOrientation(
    Facing Direction=&Lerp_Vector(
     A=&SelectVectorFromArray(
      Direct Array Index=$Particles.PartitionParticles.Partition
      Vector Selection Array=DataInterface() {propertyValues {FloatData [<1 0 0>, <1 0 0>]}}
     )
     B=&SelectVectorFromArray(
      Direct Array Index=$Particles.PartitionParticles.Partition
      Vector Selection Array=DataInterface() {propertyValues {FloatData [<1 0 0>, <1 0 0>]}}
     )
     Alpha=$User.ParaGlideLeanAlpha
    )
    Facing Interpolation=false
    Orientation Method=::NewEnumerator1
    Override Quat=false
    TransformFacing=false
    TransformUp=false
    Up Direction:Vector3f=&FN_FrontEndVector3Modifier(Gameplay=<0 0 1>, Front End=&Lerp_Vector(A=<0 0 1>, B=<0 -0.5 1>, Alpha=$User.ParaGlideLeanAlpha))
   ) {module=UpdateMeshOrientation002}
   &Timeline(
    Play Rate=&Lerp_Float(A=1, B=2, Alpha=&ClampFloat(Float=$User.ParaGlideLeanAlpha))
    Timeline End=10000000
   )
   &ScaleMeshSize(Scale Factor=&MakeVector(
     X=1
     Y=1
     Z=&FloatFromCurve(
      FloatCurve=DataInterface() {propertyValues {Curve {keys [
          {arriveTangent=5.7371, interpMode=RCIM_Cubic, leaveTangent=5.7371, tangentMode=RCTM_User, time=0, value=0}
          {arriveTangent=-0.9106, interpMode=RCIM_Cubic, leaveTangent=-0.9106, tangentMode=RCTM_User, time=0.76, value=1.025}
          <at 1 1>
         ]}}}
      CurveIndex=$Emitter.Age
      Scale Curve=$Emitter.Timeline001.NormalizedTime
     )
    ))
   &DynamicMaterialParameters(
    Index 0 Param 2=$System.Dive
    Index 0 Param 3=&RandomRangeFloat(Minimum=0, Maximum=1)
    Index 0 Param 4=&RandomRangeFloat(Minimum=0, Maximum=1)
    Module.Index0Param0Write=true
    Module.Index0Param1Write=true
    Module.Index0Param2Write=true
    Module.Index0Param3Write=true
   ) {module=DynamicMaterialParameters001}
   &SkeletalMeshLocation(
    Bone / Socket Index=$Particles.PartitionParticles.Partition
    Bone Sampling Mode=::NewEnumerator4
    MaskSpawnGroup=false
    Orientation Sampling=::NewEnumerator1
    Sampled Position Offset=&FN_FrontEndVector3Modifier(
     Gameplay=&MakeCustomVectorFromBool(
      Bool=$User.bIsFrontEndPreview
      True Vector=&SelectVectorFromArray(
       Direct Array Index=$Particles.PartitionParticles.Partition
       Vector Selection Array=DataInterface() {propertyValues {FloatData [<0 -3 1>, <0 4 1>, <0 0 0>, <0 0 0>]}}
      )
      False Vector=&Lerp_Vector(
       A=&SelectVectorFromArray(
        Direct Array Index=$Particles.PartitionParticles.Partition
        Vector Selection Array=DataInterface() {propertyValues {FloatData [<8 -3 -1>, <-8 4 1>, <0 0 0>, <0 0 0>]}}
       )
       B=&SelectVectorFromArray(
        Direct Array Index=$Particles.PartitionParticles.Partition
        Vector Selection Array=DataInterface() {propertyValues {FloatData [<25 0 0>, <-25 0 0>, <0 0 0>, <0 0 0>]}}
       )
       Alpha=$User.ParaGlideLeanAlpha
      )
     )
     Front End=&Lerp_Vector(
      A=&SelectVectorFromArray(
       Direct Array Index=$Particles.PartitionParticles.Partition
       Vector Selection Array=DataInterface() {propertyValues {FloatData [<0 -3 -1>, <0 4 1>, <0 0 0>, <0 0 0>]}}
      )
      B=&SelectVectorFromArray(
       Direct Array Index=$Particles.PartitionParticles.Partition
       Vector Selection Array=DataInterface() {propertyValues {FloatData [<25 0 0>, <-25 0 0>, <0 0 0>, <0 0 0>]}}
      )
      Alpha=$User.ParaGlideLeanAlpha
     )
    )
    Skeletal Mesh=$System.Skeletal Mesh
   ) {enabled=false}
   &SkeletalMeshLocation(
    Bone / Socket Index=$Particles.PartitionParticles.Partition
    Bone Sampling Mode=::NewEnumerator4
    MaskSpawnGroup=false
    Orientation Sampling=::NewEnumerator1
    Sampled Position Offset=<0 0 5>
    Sampled Position Offset Coordinate Space:ENiagaraSkelMeshTransforms=::NewEnumerator0
    Skeletal Mesh=$System.Skeletal Mesh
   ) {module=SkeletalMeshLocation001}
   &ScaleMeshSize(Scale Factor=&FN_FrontEndVector3Modifier(Gameplay=<0.7 0.7 0.8>, Front End=<0.65 0.65 0.75>)) {module=ScaleMeshSize001}
   &ScaleMeshSize(Scale Factor=&SelectVectorFromArray(
     Direct Array Index=$Particles.PartitionParticles.Partition
     Vector Selection Array=DataInterface() {propertyValues {FloatData [<1 1 1>, <1 1 1>, <0.5 0.5 0.4>, <0.5 0.5 0.4>]}}
    )) {module=ScaleMeshSize002}
   &ScaleColor(
    Scale Alpha=&FloatFromCurve(
     FloatCurve=DataInterface() {propertyValues {Curve {keys [<at 0 1 RCIM_Cubic>,<at 0.2 1 RCIM_Cubic>,<at 1 1>]}}}
     Scale Curve=&Lerp_Float(A=0.06, B=0.2, Alpha=$System.Dive)
    )
    Scale RGB=&MakeVectorFromLinearColorRGB(Color=&FNTimeOfDayColor(
      Night Color=<rgba 0.6 0.6 0.6 1>
      Morning Color=<rgba 0.65 0.65 0.65 1>
      Day Color=<rgba 0.9 0.9 0.9 1>
      Evening Color=<rgba 0.65 0.65 0.65 1>
     ))
    ScaleA=true
    ScaleRGB=true
   )
   &ScaleColor(Scale Alpha=&FN_FrontEndFloatModifier(Gameplay=1, Front End=0.6), ScaleA=true, ScaleRGB=true) {module=ScaleColor001}
  ]
  rendererClasses [&NiagaraMeshRendererProperties]
  renderers [{rendererClass=&NiagaraMeshRendererProperties, rendererIndex=0}]
 }
 LightningShape {
  emitterUpdateScript [
   &EmitterState(Life Cycle Mode=::NewEnumerator1, Loop Behavior=::NewEnumerator1, Loop Duration Mode=::NewEnumerator1, UseLoopDelay=false)
   &SpawnBurst_Instantaneous(Spawn Count=2, Use Spawn Probability=false, UseLoopCountLimit=false)
  ]
  particleSpawnScript [
   &PartitionParticles()
   &InitializeParticle(Color=<rgba 1 1 1 3>, UsePositionOffset=false, Write Material Random=true)
   &InitialMeshOrientation(
    Mesh Orientation Mode=::NewEnumerator7
    Rotation=&FN_FrontEndVector3Modifier(Gameplay=<0 0 0>, Front End=<0 0 0>)
    Use Rotation Vector=true
   ) {enabled=false}
   &DynamicMaterialParameters(
    Index 0 Param 1=80
    Index 0 Param 2=5
    Index 0 Param 3=&RandomRangeFloat(Minimum=-1, Maximum=1)
    Index 0 Param 4=1
    Module.Index0Param0Write=true
    Module.Index0Param1Write=true
    Module.Index0Param2Write=true
    Module.Index0Param3Write=true
   )
  ]
  particleUpdateScript [
   &ParticleState(Kill Particles When Lifetime Has Elapsed=false, Let Infinitely Lived Particles Die When Emitter Deactivates=true)
   &SkeletalMeshLocation(
    Bone / Socket Index=$Particles.PartitionParticles.Partition
    Bone Sampling Mode=::NewEnumerator4
    MaskSpawnGroup=false
    Orientation Sampling=::NewEnumerator1
    Skeletal Mesh=$System.Skeletal Mesh
   )
   &UpdateMeshOrientation(
    Facing Interpolation=false
    Orientation Method=::NewEnumerator1
    Override Quat=false
    TransformFacing=false
    TransformUp=false
    Up Direction:Vector3f=&MakeCustomVectorFromBool(
     Bool=$User.bIsFrontEnd
     True Vector=&Lerp_Vector(A=<0 0 1>, B=<0 -0.4 1>, Alpha=$User.ParaGlideLeanAlpha)
     False Vector=<1 0 1>
    )
   )
  ]
  rendererClasses [&NiagaraMeshRendererProperties]
  renderers [{rendererClass=&NiagaraMeshRendererProperties, rendererIndex=0}]
 }
 SpeedLines {
  emitterUpdateScript [
   &EmitterState()
   &SpawnRateFromMultipleSpawnGroups_2(Global Spawnrate:NiagaraFloat=&MakeCustomFloatFromBool(Bool=$User.bIsFrontEndPreview, True Float=0, False Float=&Lerp_Float(A=5, B=15, Alpha=$User.ParaGlideLeanAlpha)))
  ]
  particleSpawnScript [
   &FN_SetSpawnGroup()
   &InitializeParticle(
    Lifetime Max=0.45
    Lifetime Min=0.25
    Lifetime Mode=::NewEnumerator1
    Mesh Scale Mode=::NewEnumerator1
    Mesh Uniform Scale Max:NiagaraFloat=&RandomRangeFloat(Minimum=0.02, Maximum=0.04)
    Mesh Uniform Scale Min=0.015
    UsePositionOffset=false
    Write Material Random=true
   ) {module=InitializeParticle001}
   &SkeletalMeshLocation(
    Bone / Socket Index=$Particles.SpawnGroup
    Bone Sampling Mode=::NewEnumerator4
    Inherit Velocity=false
    MaskSpawnGroup=false
    Sampled Position Offset=<0 0 -25>
    Skeletal Mesh=$System.Skeletal Mesh
   )
   &ShapeLocation(Disc Coverage=0.619, MaskSpawnGroup=false, Override Seed=false, Shape Primitive=::NewEnumerator4, UseOwnerScale=true)
  ]
  particleUpdateScript [
   &ParticleState()
   &DynamicMaterialParameters(
    Index 0 Param 1=&FloatFromCurve(
     FloatCurve=DataInterface() {propertyValues {Curve {keys [
         {interpMode=RCIM_Cubic, leaveTangent=31.9087, tangentMode=RCTM_Break, time=0, value=0}
         {arriveTangent=31.7752, interpMode=RCIM_Cubic, leaveTangent=153.4346, tangentMode=RCTM_Break, time=0.4, value=25}
         <at 1 120>
        ]}}}
     Scale Curve=&Lerp_Float(A=1, B=1.2, Alpha=$User.ParaGlideLeanAlpha)
    )
    Index 0 Param 2=&FloatFromCurve(
     FloatCurve=DataInterface() {propertyValues {Curve {keys [<at 0.4 0>,<at 0.9 100>]}}}
     Scale Curve=&Lerp_Float(A=1, B=1.2, Alpha=$User.ParaGlideLeanAlpha)
    )
    Module.Param0WriteEnabled=true
    Module.Param1WriteEnabled=true
    Module.Param2WriteEnabled=true
    Module.Param3WriteEnabled=true
   )
  ]
  rendererClasses [&NiagaraMeshRendererProperties]
  renderers [{rendererClass=&NiagaraMeshRendererProperties, rendererIndex=0}]
 }
 Spheres {
  emitterUpdateScript [
   &EmitterState()
   &SpawnRateFromMultipleSpawnGroups_2(Global Spawnrate=15)
  ]
  particleSpawnScript [
   &FN_SetSpawnGroup()
   &InitializeParticle(
    Color=&SelectLinearColorFromArray(
     Array Sampling Mode=::NewEnumerator0
     Direct Array Index=?
     Interpolation=$System.LoopedAge
     Color Selection Array=DataInterface() {propertyValues {ColorData [<rgba 0.0459 0.0386 0.1667 1>, <rgba 0.3489 0.2624 0.8177 1>, <rgba 0.7031 0.0174 0.2835 0.979>, <rgba 0.0116 0.2817 0.6979 0.974>]}}
     Random Seed=0
    )
    Lifetime Max=0.45
    Lifetime Min=0.3
    Lifetime Mode=::NewEnumerator1
    Sprite Size Mode=::NewEnumerator3
    Uniform Sprite Size=&MakeCustomFloatFromBool(
     Bool=&SetBoolByIntComparison(A=&SpawnGroup(), B=1, Comparison Type=::NewEnumerator7)
     True Float=&RandomRangeFloat(Minimum=6, Maximum=10)
     False Float=&RandomRangeFloat(Minimum=3, Maximum=8)
    )
    UsePositionOffset=false
    Write Material Random=true
   ) {module=InitializeParticle001}
   &SkeletalMeshLocation(
    Bone / Socket Index=$Particles.SpawnGroup
    Bone Sampling Mode=::NewEnumerator4
    Inherit Velocity=false
    MaskSpawnGroup=false
    Skeletal Mesh=$System.Skeletal Mesh
   )
   &ShapeLocation(Disc Coverage=0.5, MaskSpawnGroup=false, Override Seed=false, Ring Radius=15, Shape Primitive=::NewEnumerator4, UseOwnerScale=true)
   &AddVelocity(Velocity=&RandomRangeVector(Minimum=<100 100 250>, Maximum=<-100 -100 400>))
  ]
  particleUpdateScript [
   &ParticleState()
   &DynamicMaterialParameters(
    Index 0 Param 1=&FloatFromCurve(FloatCurve=DataInterface() {propertyValues {Curve {keys [
         {interpMode=RCIM_Cubic, leaveTangent=-0.5257, tangentMode=RCTM_Break, time=0.1, value=0}
         {arriveTangent=1.6403, interpMode=RCIM_Cubic, tangentMode=RCTM_Break, time=1.2, value=1.125}
        ]}}})
    Index 0 Param 2=&FloatFromCurve(FloatCurve=DataInterface() {propertyValues {Curve {keys [<at 0.4 0>,<at 0.9 55>]}}})
    Module.Param0WriteEnabled=true
    Module.Param1WriteEnabled=false
    Module.Param2WriteEnabled=true
    Module.Param3WriteEnabled=true
   )
   &Drag(Drag=2, Use Linear Drag=true, Use Rotational Drag=true)
   &SolveForcesAndVelocity(Clamp Velocity=false, Limit Acceleration=false)
  ]
  rendererClasses [&NiagaraSpriteRendererProperties]
  renderers [{rendererClass=&NiagaraSpriteRendererProperties}]
 }
 Stars {
  emitterUpdateScript [
   &EmitterState(Life Cycle Mode=::NewEnumerator1, Loop Behavior=::NewEnumerator1, Loop Duration Mode=::NewEnumerator1, UseLoopDelay=false)
   &SpawnBurst_Instantaneous(Spawn Count=4, Use Spawn Probability=false, UseLoopCountLimit=false)
  ]
  particleSpawnScript [
   &PartitionParticles(Number Of Partitions=4)
   &InitializeParticle(
    Color Mode=::NewEnumerator0
    Mesh Scale Mode=::NewEnumerator3
    Mesh Uniform Scale=&SelectFloatFromArray(Direct Array Index=$Particles.PartitionParticles.Partition, Float Selection Array=DataInterface() {propertyValues {FloatData [0.7,0.7,1,1]}})
    UsePositionOffset=false
    Write Material Random=true
   )
  ]
  particleUpdateScript [
   &ParticleState(
    Kill Particles When Lifetime Has Elapsed=false
    Let Infinitely Lived Particles Die When Emitter Deactivates=true
    Loop Particles Lifetime=true
   )
   &SkeletalMeshLocation(
    Bone / Socket Index=&SelectIntFromArray(Direct Array Index=$Particles.PartitionParticles.Partition, Int Selection Array=DataInterface() {propertyValues {IntData [0,1,0,1]}})
    Bone Sampling Mode=::NewEnumerator4
    MaskSpawnGroup=false
    Orientation Sampling=::NewEnumerator1
    Sampled Position Offset=&SelectVectorFromArray(
     Direct Array Index=$Particles.PartitionParticles.Partition
     Vector Selection Array=DataInterface() {propertyValues {FloatData [<0 0 8>, <0 0 8>, <0 0 -5>, <0 0 -5>]}}
    )
    Sampled Position Offset Coordinate Space:ENiagaraSkelMeshTransforms=::NewEnumerator4
    Skeletal Mesh=$System.Skeletal Mesh
   )
   &UpdateMeshOrientation(
    Rotation Rate=&Multiply_Float(
     A=&SelectFloatFromArray(
      Direct Array Index=$Particles.PartitionParticles.Partition
      Float Selection Array=DataInterface() {propertyValues {FloatData [0.5,0.5,-0.8,-0.8]}}
     )
     B=&Lerp_Float(A=1, B=1.5, Alpha=$User.ParaGlideLeanAlpha)
    )
    Rotation Vector=<0 0 1>
   )
  ]
  rendererClasses [&NiagaraMeshRendererProperties]
  renderers [{rendererClass=&NiagaraMeshRendererProperties, rendererIndex=0}]
 }
}
summary {
 systemName=NS_Contrail_SparkleChop
 emitters [
  {emitterName=Spheres, rendererClasses [&NiagaraSpriteRendererProperties]}
  {emitterName=SpeedLines, rendererClasses [&NiagaraMeshRendererProperties]}
  {emitterName=DiscoStreaks, rendererClasses [&NiagaraMeshRendererProperties]}
  {emitterName=Decal, rendererClasses [&NiagaraSpriteRendererProperties]}
  {emitterName=LightRay, rendererClasses [&NiagaraMeshRendererProperties]}
  {emitterName=LightningRay, rendererClasses [&NiagaraMeshRendererProperties]}
  {emitterName=LightningShape, rendererClasses [&NiagaraMeshRendererProperties]}
  {emitterName=Flare, rendererClasses [&NiagaraSpriteRendererProperties, &NiagaraLightRendererProperties]}
  {emitterName=Stars, rendererClasses [&NiagaraMeshRendererProperties]}
 ]
 userVariables [
  User.bIsFrontEnd {type=NiagaraBool}
  User.bIsFrontEndPreview {type=NiagaraBool}
  User.ParaGlideLeanAlpha {type=NiagaraFloat, defaultValue=0, description=""}
  User.RibbonIDOffset {type=NiagaraInt32}
  User.SkeletalMesh {
   type=NiagaraDataInterfaceSkeletalMesh
   defaultValue {
    struct=NiagaraExt_VariableValue_DataInterface
    value {dataInterface=&self:NiagaraDataInterfaceSkeletalMesh_3}
   }
  }
 ]
}
